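# Physics asset for the ClaimReflect drum kit cosmetic, giving capsule bodies to its handle, kick drum, snare, turret and a dangling Cortana chip.
# The chip hangs from snare_shake on a five-link string chain whose joints lock twist and limit swing.
# The turret joins the kick drum with both swings locked, leaving only twist.
# Snare_shake and kick drum each join the handle with all angular motion free.
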PhysicsAsset /SparksCosmetics/Instrument/DrumKitHighVert/ClaimReflect/Meshes/ClaimReflect_DrumKit_Physics
bodies {
 dyn_cortana_chip [{shapeType=Capsule, length=1.5, radius=3.5538, shapeName=dyn_cortana_chip_capsule, center=<2.6 0 0>, rotation=<rot -90 0.0001 0>}]
 dyn_string_01 [{shapeType=Capsule, length=0.5, radius=1.5, shapeName=dyn_string_01_capsule, rotation=<rot 90 0 0>}]
 dyn_string_02 [{shapeType=Capsule, length=1, radius=1.5, shapeName=dyn_string_02_capsule, rotation=<rot 90 0 0>}]
 dyn_string_03 [{shapeType=Capsule, length=1, radius=1.5, shapeName=dyn_string_03_capsule, rotation=<rot 90 0 0>}]
 dyn_string_04 [{shapeType=Capsule, length=1, radius=1.5, shapeName=dyn_string_04_capsule, rotation=<rot 90 0 0>}]
 dyn_string_05 [{shapeType=Capsule, length=0, radius=1.5, shapeName=dyn_string_05_capsule, rotation=<rot 90 0 0>}]
 handle [{shapeType=Capsule, length=0.505, radius=0.505, shapeName=handle_capsule, rotation=<rot 0 -90 0>}]
 kickdrum [{shapeType=Capsule, length=10, radius=10, shapeName=kickdrum_capsule, center=<-30.3179 -0.9748 33.5105>}]
 snare_shake [{shapeType=Capsule, length=0, radius=10, shapeName=snare_shake_capsule, rotation=<rot 0 -1.0476 -63.976>}]
 turret [{
   shapeType=Capsule
   length=11.8933
   radius=4.6839
   shapeName=turret_capsule
   center=<-0.5762 0.0244 0.0148>
   rotation=<rot 0 -90.0547 -89.9667>
  }]
}
constraints [
 {
  bone1Name=turret
  bone2Name=kickdrum
  swing1LimitDegrees=45
  swing1Motion=Locked
  swing2LimitDegrees=45
  swing2Motion=Locked
  twistLimitDegrees=180
 }
 {bone1Name=dyn_string_01, bone2Name=snare_shake, swing1LimitDegrees=45, swing2LimitDegrees=20, twistLimitDegrees=45, twistMotion=Locked}
 {bone1Name=dyn_string_02, bone2Name=dyn_string_01, swing1LimitDegrees=45, swing2LimitDegrees=20, twistLimitDegrees=45, twistMotion=Locked}
 {bone1Name=dyn_string_03, bone2Name=dyn_string_02, swing1LimitDegrees=45, swing2LimitDegrees=20, twistLimitDegrees=45, twistMotion=Locked}
 {bone1Name=dyn_string_04, bone2Name=dyn_string_03, swing1LimitDegrees=45, swing2LimitDegrees=45, twistLimitDegrees=45, twistMotion=Locked}
 {bone1Name=dyn_string_05, bone2Name=dyn_string_04, swing1LimitDegrees=45, swing2LimitDegrees=30, twistLimitDegrees=45, twistMotion=Locked}
 {bone1Name=dyn_cortana_chip, bone2Name=dyn_string_05, swing1LimitDegrees=40, swing2LimitDegrees=45, twistLimitDegrees=45, twistMotion=Locked}
 {
  bone1Name=snare_shake
  bone2Name=handle
  swing1LimitDegrees=45
  swing1Motion=Free
  swing2LimitDegrees=45
  swing2Motion=Free
  twistLimitDegrees=45
  twistMotion=Free
 }
 {
  bone1Name=kickdrum
  bone2Name=handle
  swing1LimitDegrees=45
  swing1Motion=Free
  swing2LimitDegrees=45
  swing2Motion=Free
  twistLimitDegrees=45
  twistMotion=Free
 }
]
