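# Post-style Anim Blueprint for the ClaimReflect drum kit cosmetic, adding physics and procedural spin to the input pose.
# RigidBody simulates relative to the handle bone, then an additive bone transform yaws the turbine by an angle accumulated each update.
# Event graph logic also derives turret spin from music clock, emote and front-end montage state.
# The turret rotation, AnimDynamics and additive turret loop do not feed the output pose.
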AnimBlueprint /SparksCosmetics/Instrument/DrumKitHighVert/ClaimReflect/Meshes/ClaimReflect_DrumKit_AnimBP
target_skeleton=&/ClaimReflect_DrumKit_Skeleton
graphs [
 AnimGraph {nodes [
   Anim Dynamics - Bone: turret(Alpha=1, ExternalForce="0,0,0") {
    simulationSpace=BoneRelative
    alphaBoolBlend {blendOption=Linear}
    boundBone {boneName=turret}
    physicsBodyDefinitions [{
      boundBone {boneName=turret}
      boxExtents=<10 10 10>
      constraintSetup {angularLimitsMax=<120 0 0>, angularLimitsMin=<-90 0 0>, angularTarget=<0 0 0>}
      localJointOffset=<1 0 0>
     }]
   }
   Apply Additive(
    Base=Component To Local()
    Additive="Sequence Player (additive) 'ClaimReflect_DrumKit_Turret_Loop'"() {playRate=6, sequence=/SparksCosmetics/Instrument/DrumKitHighVert/ClaimReflect/Anim/ClaimReflect_DrumKit_Turret_Loop}
   ) {alphaBoolBlend {blendOption=Linear}}
   n1 = RigidBody(ComponentPose=Local To Component(LocalPose=Input Pose())) {
    bOverrideWorldGravity=true
    overridePhysicsAsset=&/ClaimReflect_DrumKit_Physics
    simulationSpace=BaseBoneSpace
    alphaBoolBlend {blendOption=Linear}
    baseBoneRef {boneName=handle}
    componentLinearAccScale=<0.6 0.6 0.6>
    componentLinearVelScale=<0.6 0.6 0.6>
    overrideWorldGravity {z=-980}
    simSpaceSettings {
     dampingAlpha=0
     maxAngularAcceleration=80
     maxAngularVelocity=6
     maxLinearAcceleration=800
     maxLinearVelocity=500
     worldAlpha=1
     externalLinearDragV=<0.1 0.1 0.1>
    }
   }
   n2 = "Transform (Modify) Bone - Bone: turbine"!(
    ComponentPose=@n1
    Alpha=1
    Translation="0,0,0"
    Rotation_Roll:float=0
    Rotation_Pitch:float=0
    Rotation_Yaw:float=Get!() -> (rotation of turbine_Roll:float, rotation of turbine_Pitch:float, rotation of turbine_Yaw:float).rotation of turbine_Yaw
    Scale="1,1,1"
   ) -> (Pose) {
    rotationMode=BMM_Additive
    rotationSpace=BCS_BoneSpace
    translationSpace=BCS_BoneSpace
    alphaBoolBlend {blendOption=Linear}
    boneToModify {boneName=turbine}
   }
   Output Pose(Result=Component To Local(ComponentPose=@n2))
  ]}
 EventGraph {nodes [
   Event BlueprintUpdateAnimation()
   > n1 = Sequence!(execute) -> (then_0, then_1, then_2:Exec, then_3:Exec, then_4:Exec, then_5:Exec)
   > n2 = Branch(Condition=Get bIsInFrontEnd())
   > Set RotationSpeed(RotationSpeed:double=200) -> (Output_Get:double)
   n3 = GetWorldDeltaSeconds!() -> (ReturnValue:double)
   n4 = Get RotationSpeed() -> (RotationSpeed:double)
   Set RotationSpeed(execute=@n2.else, RotationSpeed="Clamp (Float)"(Value=FInterpTo(Current=@n4, Target=200, DeltaTime=@n3, InterpSpeed=1), Min=150, Max=200))
   n5 = GetWorldDeltaSeconds()
   n6 = Get RotationSpeed()
   n7 = Set CurrentAngle(
    execute=@n1.then_1
    CurrentAngle:double=float + float(A=Get CurrentAngle() -> (CurrentAngle:double), B=float * float(A=@n5, B=@n6) {is=title K2Node_PromotableOperator})
   ) -> (Output_Get:double)
   > n8 = Set CurrentAngle(CurrentAngle="% (Float)"!(A:double=@n7, B:double=360) -> (ReturnValue:double))
   > Set rotation of turbine(rotation of turbine:Rotator=MakeRotator(Yaw=@n8)) -> (Output_Get:Rotator)
   Reroute Node()
   > n9 = Branch(Condition=Get Turrnet Spin() -> (Turrnet Spin:bool))
   n10 = GetWorldDeltaSeconds()
   n11 = Get Turret rotation speed() -> (Turret rotation speed:double)
   > Set Turret rotation speed(Turret rotation speed:double="Clamp (Float)"(Value=FInterpTo(Current=@n11, Target=1000, DeltaTime=@n10, InterpSpeed=3), Min=400, Max=1000)) -> (Output_Get:double)
   Set Turret rotation speed(execute=@n9.else, Turret rotation speed=0)
   Reroute Node()
   n12 = GetWorldDeltaSeconds()
   n13 = Get Turret rotation speed()
   > n14 = Set Turret Angle(Turret Angle:double=float + float(A=Get Turret Angle() -> (Turret Angle:double), B=float * float(A=@n12, B=@n13))) -> (Output_Get:double)
   > n15 = Set Turret Angle(Turret Angle="% (Float)"(A=@n14, B=360))
   > Set rotation of turret(rotation of turret:Rotator=MakeRotator(Yaw=@n15)) -> (Output_Get:Rotator)
   Reroute Node()
   > n16 = Is Valid(InputObject=GetHighestPriorityClock!(self:Music Event Subsystem Object Reference=Get MusicEventSubsystem() -> (ReturnValue:Music Event Subsystem Object Reference) {is=title K2Node_GetSubsystem}) -> (ReturnValue:Music Clock Object Reference))
   n17 = Branch(execute=@n16.Is Not Valid, Condition=Get bIsPlayingEmote())
   > Set Turrnet Spin(Turrnet Spin:bool=true) -> (Output_Get:bool)
   Set Turrnet Spin(execute=@n17.else, Turrnet Spin=false)
   Set Turrnet Spin(execute=@n16.Is Valid, Turrnet Spin=true)
   Reroute Node()
   n18 = GetAnimInstance(self=GetOwnerSkeletalMeshComponent())
   > Is Valid(InputObject=@n18)
   > n19 = Set FE_Idle_Playing(FE_Idle_Playing:bool=Montage_IsActive!(
     self:Anim Instance Object Reference=@n18
     Montage:Anim Montage Object Reference=/SparksCosmetics/FrontendPreview/Poses/Drum_SpawnInstrument_Pose_FE_M
    ) -> (ReturnValue:bool)) -> (Output_Get:bool)
   > n20 = Branch(Condition=@n19)
   > Set Turrnet Spin(Turrnet Spin=false)
   Set Turrnet Spin(execute=@n20.else, Turrnet Spin=true)
   TryGetPawnOwner()
   Is MIDI Song Position Valid!(SongPos:"MIDI Song Position Structure (by ref)"=GetSongPos!(self:Music Clock Object Reference, Timebase:ECalibratedMusicTimebase Enum=VideoRenderTime) -> (ReturnValue:"MIDI Song Position Structure (by ref)")) -> (ReturnValue:bool)
  ]}
]
variables [
 animation plus delay {type=boolean}
 CurrentAngle {type=number}
 FE_Idle_Playing {type=boolean}
 rotation of turbine {type=Rotator}
 rotation of turret {type=Rotator}
 RotationSpeed {type=number}
 Turret Angle {type=number}
 Turret rotation speed {type=number}
 Turrnet Spin {type=boolean}
]
